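# Editor utility that supplies helper functions for the asset tag debug text component to look up metadata tag strings on assets and actors.
# TryGetMetadataStringfromAsset reads a named metadata tag from a valid asset and reports whether the string is non-empty.
# TryGetTextureDataStringFromActor reads an actor's TextureData array by property name through GetEditorProperty, then returns the first tag string found among those Building Texture Data assets.
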EditorUtilityBlueprint /Game/Environments/Blueprints/AssetTagDebugTextComponent/EUO_ATTC_EditorUtils
graphs [
 EventGraph {links [], nodes []}
 TryGetMetadataStringfromAsset {nodes [
   n1 = TryGetMetadataStringfromAsset!() -> (then, Asset:Object Reference, Tag:Name=None) {is=title K2Node_FunctionEntry}
   > Is Valid(InputObject=@n1)
   > Branch(Condition="Not Equal (Name)"(A=@n1.Tag) -> (ReturnValue=~))
   > n2 = GetMetadataTag!(execute, Object:Object Reference=@n1, Tag:Name=@n1.Tag) -> (then, ReturnValue:String)
   > Return Node!(execute, FoundMetadata:bool=NOT Boolean(A=IsEmpty(InString=@n2)), String:String=@n2)
  ]}
 TryGetTextureDataStringFromActor {nodes [
   n1 = TryGetTextureDataStringFromActor!() -> (then, Actor:Actor Object Reference, Tag:Name=None) {is=title K2Node_FunctionEntry}
   > n2 = GetEditorProperty!(execute:Exec, Object:Object Reference=@n1, PropertyName:Name=TextureData) -> (then:Exec, PropertyValue:Array of Building Texture Data Object References, ReturnValue:bool)
   > Set l_datas(l_datas:Array of Building Texture Data Object References=@n2) -> (Output_Get:Array of Building Texture Data Object References)
   > n3 = Branch(Condition=@n2.ReturnValue)
   > n4 = For Each Loop(Array:Array of Building Texture Data Object References=Get l_datas() -> (l_datas:Array of Building Texture Data Object References)) -> (Array Element:Building Texture Data Object Reference)
   Return Node!(execute=[@n3.else, @n4.Completed], Valid:bool=false, String)
   > n5 = TryGetMetadataStringfromAsset!(execute, self, Asset:Object Reference=@n4, Tag:Name=Get Tag() -> (Tag:Name=None)) -> (then, FoundMetadata:bool, String:String) {is=title K2Node_CallFunction}
   > Branch(Condition=@n5)
   > Return Node(String=@n5.String)
  ]}
]
variables []
